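; Gameplay cue notify for the theme park coaster grind rail that plays a tunnel whoosh sound when activated on a fast-moving local grinder.
; On activation it skips mobile and Switch, then requires the target pawn's grinding component to report grinding and the pawn to be locally controlled.
; Above a speed threshold, taken from the larger absolute X or Y velocity, it spawns the 2D sound and maps that speed to its volume parameter once.
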(event EventBeginPlay)

(event Collision|EventActorBeginOverlap
  (params (OtherActor :type Actor%20Object%20Reference)))

(event EventTick
  (params (DeltaSeconds :type Float%20%28single-precision%29)))

(event GameplayCueNotify|EventHandleGameplayCue
  (params (MyTarget :type Actor%20Object%20Reference) (EventType :type EGameplayCueEvent%20Enum :default "OnActive") (Parameters :type Gameplay%20Cue%20Parameters%20Structure%20%28by%20ref%29))
  (switch Utilities|FlowControl|Switch|SwitchonEGameplayCueEvent EventType
    (:OnActive
      (if (not (or (Fortnite|IsMobilePlatform) (TouchPlatform|IsNintendoSwitchorNintendoSwitchPreview)))
        (bind _asplayer_pawn_athena (Utilities|Casting|CastToPlayerPawn_Athena MyTarget))
        (if (Class|FortPawnComponentGrindingProvider|IsGrinding (Actor|GetComponentbyClass _asplayer_pawn_athena "/GrindRail/Blueprints/BP_FortPawnGrindingProvider.BP_FortPawnGrindingProvider_C"))
          (if (Pawn|IsLocallyControlled _asplayer_pawn_athena)
            (if (> (Math|Float|Max(Float) (Math|Float|Absolute(Float) (.x (Transformation|GetVelocity _asplayer_pawn_athena))) (Math|Float|Absolute(Float) (.y (Transformation|GetVelocity _asplayer_pawn_athena)))) 8e+02)
              (bind _returnvalue (Audio|SpawnSound2D _asplayer_pawn_athena "/GrindRail/Audio/Coaster/MSS_GrindRail_Coaster_TunnelWhoosh.MSS_GrindRail_Coaster_TunnelWhoosh"))
              (bind _output_get (Variables|Default|SetWhooshAudioComp _returnvalue))
              (Audio|Parameter|SetFloatParameter _output_get "WhooshVol" (Math|Float|MapRangeClamped (Math|Float|Max(Float) (Math|Float|Absolute(Float) (.x (Transformation|GetVelocity _asplayer_pawn_athena))) (Math|Float|Absolute(Float) (.y (Transformation|GetVelocity _asplayer_pawn_athena)))) 8e+02 2.2e+03 0.2 1.0) :!class /Script/AudioExtensions.AudioParameterControllerInterface))))))))

(fn ConstructionScript
  (|CallParentFunction))
